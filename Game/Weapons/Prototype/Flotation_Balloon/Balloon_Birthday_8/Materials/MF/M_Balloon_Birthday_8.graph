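# Subsurface material-attributes material for the Birthday 8 variant of the Flotation Balloon gadget, flagged for skeletal meshes.
# MF_Balloon_Birthday_8 builds the base; MF_Slurp, masked by an FX map channel, supplies emissive to an emissive layer blend.
# A separate stencil branch selects between two rotated, offset UV sets by V half, scales them and clips the sampled mask.
# That stencil result has no shown connection to the rest of the graph.
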Material /Game/Weapons/Prototype/Flotation_Balloon/Balloon_Birthday_8/Materials/MF/M_Balloon_Birthday_8
graph {
 properties {
  bUseMaterialAttributes=true
  bUsedWithMeshDeformer=true
  bUsedWithSkeletalMesh=true
  pixelDepthOffsetMode=PDOM_Legacy
  shadingModel=MSM_Subsurface
  usedShadingModels=Subsurface
 }
 nodes [
  n1 = MaterialFunctionCall() {materialFunction=&/MF_Balloon_Birthday_8}
  n2 = TextureSampleParameter2D() {parameterName=FX map, samplerType=SAMPLERTYPE_Masks, texture=/Game/Weapons/Prototype/Flotation_Balloon/Texture/T_Gadget_FlotationBalloon_FX}
  n3 = StaticComponentMaskParameter(None=@n2.RGB) {defaultR=true, group=Skin FX, parameterName=Slurp Channel}
  MaterialFunctionCall!(
   Material=@n1.Result
   Emissive=GetMaterialAttributes(None=MaterialFunctionCall!(Alpha=@n3, AO, BubblesMask=@n3, SlurpNormalDistortionMask, Pre Skinned Position, Material=@n1.Result) {materialFunction=/Game/Weapons/WeaponSkins/MaterialFunction/MF_Slurp}.Result) {attributeGetTypes [#1]}.Emissive Color
  ) {materialFunction=/Engine/Functions/MaterialLayerFunctions/MatLayerBlend_Emissive}
  n4 = TextureCoordinate()
  n5 = Rotator(Coordinate=@n4, Time=ScalarParameter() {defaultValue=6.28, parameterName=RotateStencil_1})
  n6 = AppendVector(
   A=ScalarParameter() {defaultValue=0.24, parameterName=StencilOffset_X_1}
   B=ScalarParameter() {defaultValue=-0.18, parameterName=StencilOffset_Y_1}
  )
  n7 = Rotator(Coordinate=@n4, Time=ScalarParameter() {defaultValue=-6.28, parameterName=RotateStencil})
  n8 = AppendVector(
   A=ScalarParameter() {defaultValue=0.26, parameterName=StencilOffset_X}
   B=ScalarParameter() {defaultValue=0.07, parameterName=StencilOffset_Y}
  )
  n9 = Add(A=ComponentMask(None=TextureCoordinate()) {g=true}) {constB=-0.5}
  n10 = MaterialFunctionCall(
   UVs=LinearInterpolate(A=Subtract(A=@n5, B=@n6) {constA=0.5, constB=0.5}, B=Subtract(A=@n7, B=@n8) {constA=0.5, constB=0.5}, Alpha=Ceil(None=@n9))
   Texture Scale=AppendVector(A=ScalarParameter() {defaultValue=0.25, parameterName=StencilScale_X}, B=ScalarParameter() {defaultValue=0.31, parameterName=StencilScale_Y})
  ) {materialFunction=&ScaleUVsByCenter}
  Multiply(
   A=TextureSampleParameter2D(UVs=@n10.UVs) {parameterName=Stencil, samplerType=SAMPLERTYPE_Masks, texture=&T_BlackMask}.R
   B=@n10.0-1 mask
  )
 ]
}
